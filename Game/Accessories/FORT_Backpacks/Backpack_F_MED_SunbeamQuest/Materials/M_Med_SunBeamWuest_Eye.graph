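# Skinned, lit character material for the eye of the SunbeamQuest backpack, animating the eye texture with a timed vertical UV scroll.
# Time since a timestamp parameter, named as Blueprint-supplied, is normalized by a duration, inverted and remapped through a curve atlas to offset halved-V UV1.
# The sampled texture replaces Base Color between the shared character skin FX and post-FX functions.
# The EyeSpin A/B falloff lerps feed nothing.
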Material /Game/Accessories/FORT_Backpacks/Backpack_F_MED_SunbeamQuest/Materials/M_Med_SunBeamWuest_Eye
graph {
 properties {
  bUseMaterialAttributes=true
  bUsedWithMeshDeformer=true
  bUsedWithSkeletalMesh=true
  pixelDepthOffsetMode=PDOM_Legacy
  refractionMethod=RM_IndexOfRefraction
  usedShadingModels=Default Lit
 }
 nodes [
  n1 = ScalarParameter() {defaultValue=-5.5, group=SBeam_Eye_Rotation, parameterName=EyeSpin_A_End}
  n2 = Time()
  n3 = ScalarParameter() {defaultValue=9999999, group=SBeam_Eye_Rotation, parameterName=TimeFrom-BP----------------, sortPriority=600}
  n4 = ScalarParameter() {defaultValue=1, group=SBeam_Eye_Rotation, parameterName=EyeSpin_B_DElay}
  n5 = Divide(A=Subtract(A=@n2, B=Add(A=@n3, B=@n4)), B=ScalarParameter() {defaultValue=1, group=SBeam_Eye_Rotation, parameterName=EyeSpin_B_Duration})
  LinearInterpolate(
   A=ScalarParameter() {defaultValue=4, group=SBeam_Eye_Rotation, parameterName=EyeSpin_B_End}
   B=@n1
   Alpha=Power(Base=Saturate(None=@n5), Exp=ScalarParameter() {defaultValue=0.25, group=SBeam_Eye_Rotation, parameterName=EyeSpin_B_Falloff})
  )
  n6 = Saturate(None=Divide(A=Subtract(A=@n2, B=@n3), B=ScalarParameter() {defaultValue=10, group=SBeam_Eye_Rotation, parameterName=EyeSpin_A_Duration}))
  LinearInterpolate(A=@n1, B=ScalarParameter() {group=SBeam_Eye_Rotation, parameterName=EyeSpin_A}, Alpha=OneMinus(None=@n6))
  n7 = MaterialFunctionCall() {materialFunction=&BaseFN_CharacterFunction_PreFX}
  n8 = MaterialFunctionCall(
   PreskinnedPosition=@n7.Pre skinned Local Position
   Gmap=@n7.Pre skinned Local Position
   MaterialAttributes=@n7.Pre skinned Local Position
   AO=@n7.Pre skinned Local Position
  ) {materialFunction=&CharacterSkinFX}
  n9 = TextureCoordinate() {coordinateIndex=1}
  n10 = Constant2Vector() {g=2, r=1}
  n11 = Constant()
  n12 = CurveAtlasRowParameter(CurveTime=OneMinus(None=@n6)) {
   atlas=/Game/Accessories/FORT_Backpacks/Backpack_F_MED_SunbeamQuest/Textures/CA_SunBeamQuest
   curve=/Game/Accessories/FORT_Backpacks/Backpack_F_MED_SunbeamQuest/Textures/CC_AnimationTime
   group=SBeam_Eye_Rotation
   parameterName=EyeSpin_Time_Curve
  }
  n13 = TextureSampleParameter2D(UVs=Add(A=Divide(A=@n9, B=@n10), B=AppendVector(A=@n11, B=@n12.R))) {
   group=SBeam_Eye_Rotation
   parameterName=Eye__Tex
   texture=/Game/Accessories/FORT_Backpacks/Backpack_F_MED_SunbeamQuest/Textures/T_SunbeamQuest_Eye_BC
  }
  n14 = MaterialFunctionCall(
   AO=@n7.AO
   Distance Blend=@n7.AO
   Fresnel=@n7.AO
   Gmap=@n7.AO
   Pre Skinned Local Position=@n7.AO
   Material=SetMaterialAttributes!(
    MaterialAttributes=GetMaterialAttributes(None=@n8.Result) {attributeGetTypes [#1]}.MaterialAttributes
    Base Color=@n13.RGB
   ) {attributeSetTypes [#1]}
  ) {materialFunction=&BaseFN_CharacterFunction_PostFX}
  MaterialFunctionCall(In=@n14.Result) {materialFunction=&CINE_MF_CharPostFX}
 ]
}
