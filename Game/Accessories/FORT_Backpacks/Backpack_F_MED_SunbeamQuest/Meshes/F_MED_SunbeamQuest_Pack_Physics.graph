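# Physics asset for the SunbeamQuest backpack mesh, setting up collision and joint limits for its dyn_ tail and fin bones.
# Capsule bodies cover a tail base, a three-link tail chain and a left and right fin.
# The tail chains from the base with narrow twist and swing limits that widen toward the tip.
# Both fins attach to the tail base with limited swing and locked twist.
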PhysicsAsset /Game/Accessories/FORT_Backpacks/Backpack_F_MED_SunbeamQuest/Meshes/F_MED_SunbeamQuest_Pack_Physics
bodies {
 dyn_fin_l1 [{shapeType=Capsule, length=3, radius=4, center=<5.5366 0 0>, rotation=<rot 90 0 0>}]
 dyn_fin_r1 [{shapeType=Capsule, length=3, radius=4, center=<-5.5366 0 0>, rotation=<rot 90 0 0>}]
 dyn_tale_1 [{shapeType=Capsule, length=2.1726, radius=8.8284, center=<2.1656 0 0>, rotation=<rot 90 0 0>}]
 dyn_tale_2 [{shapeType=Capsule, length=2.5909, radius=4.4091, center=<4.1245 0 0>, rotation=<rot 90 0 0>}]
 dyn_tale_3 [{shapeType=Capsule, length=1, radius=6, center=<4.1245 0 0>, rotation=<rot 90 0 0>}]
 dyn_talebase [{shapeType=Capsule, length=10.73, radius=10, center=<1.9845 0 4.1221>, rotation=<rot 70.8761 -179.9999 -179.9999>}]
}
constraints [
 {bone1Name=dyn_tale_2, bone2Name=dyn_tale_1, swing1LimitDegrees=15, swing2LimitDegrees=15, twistLimitDegrees=1}
 {bone1Name=dyn_tale_3, bone2Name=dyn_tale_2, swing1LimitDegrees=20, swing2LimitDegrees=20, twistLimitDegrees=1}
 {bone1Name=dyn_tale_1, bone2Name=dyn_talebase, swing1LimitDegrees=5, swing2LimitDegrees=5, twistLimitDegrees=1}
 {bone1Name=dyn_fin_r1, bone2Name=dyn_talebase, swing1LimitDegrees=10, swing2LimitDegrees=10, twistLimitDegrees=45, twistMotion=Locked}
 {bone1Name=dyn_fin_l1, bone2Name=dyn_talebase, swing1LimitDegrees=10, swing2LimitDegrees=10, twistLimitDegrees=45, twistMotion=Locked}
]
